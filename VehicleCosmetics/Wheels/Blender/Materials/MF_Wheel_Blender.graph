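# Material function for the Blender vehicle cosmetic wheel, outputting material attributes that blend rim and tire surfaces.
# Static switches control masked custom coloring, emissive masks and roughness sources, with channel pickers and inversion for the masks.
# Rim and tire diffuse, AO, normal, metallic, specular and roughness are lerped, with Fresnel darkening the diffuse and tread direction parameters flipping tire UVs.
# Vertex green masks a TirePop world position offset along the vector to the object position.
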MaterialFunction /VehicleCosmetics/Wheels/Blender/Materials/MF_Wheel_Blender
graph {nodes [
  Composite() {subgraphName=Darkening Fresnel}
  FunctionOutput(None=MakeMaterialAttributes!(
    BaseColor=NamedRerouteUsage()
    Metallic=NamedRerouteUsage()
    Specular=NamedRerouteUsage()
    Roughness=NamedRerouteUsage()
    Anisotropy
    EmissiveColor=NamedRerouteUsage()
    Opacity
    OpacityMask
    Normal=NamedRerouteUsage()
    Tangent
    WorldPositionOffset=NamedRerouteUsage()
    SubsurfaceColor
    ClearCoat
    ClearCoatRoughness
    AmbientOcclusion
    Refraction
    CustomizedUVs_0
    CustomizedUVs_1
    CustomizedUVs_2
    CustomizedUVs_3
    CustomizedUVs_4
    CustomizedUVs_5
    CustomizedUVs_6
    CustomizedUVs_7
    PixelDepthOffset
    ShadingModel
    Displacement
   ))
  n1 = NamedRerouteUsage()
  n2 = Multiply(A=NamedRerouteUsage(), B=ScalarParameter() {defaultValue=1, group=Tire Color Control, parameterName=Tire_CustomColorBrightness})
  n3 = TextureSampleParameter2D() {
   group=Tire Color Control
   parameterName=Tire_CustomColorMask
   samplerType=SAMPLERTYPE_LinearColor
   texture=/VehicleCosmetics/Wheels/Blender/Textures/T_Rim_Blender_RGB
  }
  n4 = ChannelMaskParameter(None=@n3.RGBA) {group=Tire Color Control, parameterName=Tire_CustomColorChannelPicker, sortPriority=2}
  n5 = StaticSwitchParameter(True=OneMinus(None=@n4.R), False=@n4.R) {group=Tire Color Control, parameterName=Tire_InvertColorMask, sortPriority=2}
  n6 = LinearInterpolate(A=@n1, B=@n2, Alpha=@n5)
  n7 = NamedRerouteUsage()
  n8 = LinearInterpolate(A=@n7, B=Multiply(A=@n7, B=@n2), Alpha=@n5)
  NamedRerouteDeclaration(None=StaticSwitchParameter(
    True=StaticSwitchParameter(True=@n6, False=@n1) {group=Tire Color Control, parameterName=TireColorIsMasked}
    False=StaticSwitchParameter(True=@n8, False=@n7) {group=Tire Color Control, parameterName=TireColorIsMasked}
   ) {group=Tire Color Control, parameterName=ColorTireDiffuse}) {name=ColorTireDiffuse, nodeColor=<rgba 0.4297 0 1 1>}
  n9 = NamedRerouteUsage()
  n10 = TextureSampleParameter2D() {
   group=Rim Color Control
   parameterName=Rim_CustomColorMask
   samplerType=SAMPLERTYPE_LinearColor
   texture=&T_Rim_Blender_RGB
  }
  n11 = ChannelMaskParameter(None=@n10.RGBA) {group=Rim Color Control, parameterName=Rim_CustomColorChannelPicker, sortPriority=2}
  n12 = StaticSwitchParameter(True=OneMinus(None=@n11.R), False=@n11.R) {group=Rim Color Control, parameterName=Rim_InvertColorMask, sortPriority=2}
  n13 = Multiply(
   A=@n12
   B=Multiply(A=NamedRerouteUsage(), B=ScalarParameter() {defaultValue=1, group=Rim Color Control, parameterName=Rim_CustomColorBrightness})
  )
  n14 = LinearInterpolate(A=@n9, B=@n13, Alpha=@n12)
  NamedRerouteDeclaration(None=StaticSwitchParameter(True=StaticSwitchParameter(True=@n14, False=@n13) {group=Rim Color Control, parameterName=RimColorIsMasked}, False=@n9) {group=Rim Color Control, parameterName=ColorRimDiffuse}) {name=ColorRimDiffuse, nodeColor=<rgba 0 0.8828 1 1>}
  n15 = NamedRerouteUsage()
  n16 = ScalarParameter() {defaultValue=1, parameterName=RimAOMultiplier}
  n17 = StaticSwitchParameter(True=NamedRerouteUsage(), False=Constant() {r=1}) {defaultValue=true, group=Tire Properties, parameterName=TireHasAO}
  n18 = ScalarParameter() {defaultValue=1, parameterName=TireAOMultiplier}
  n19 = LinearInterpolate(A=Saturate(None=Multiply(A=@n15, B=@n16)), B=Saturate(None=Multiply(A=@n17, B=@n18)), Alpha=NamedRerouteUsage())
  NamedRerouteDeclaration(None=@n19) {name=AOxDiffuse, nodeColor=<rgba 1 0.2344 0 1>}
  n20 = LinearInterpolate(Alpha=ScalarParameter() {defaultValue=1, parameterName=TreadUVDirection_X}) {constA=-1}
  n21 = LinearInterpolate(Alpha=ScalarParameter() {defaultValue=1, parameterName=TreadUVDirection_Y}) {constA=-1}
  n22 = TextureSampleParameter2D(UVs=Multiply(A=TextureCoordinate(), B=AppendVector(A=@n20, B=@n21))) {group=Tire Properties, parameterName=TireDiffuse, texture=/VehicleCosmetics/Wheels/_SharedTextures/Tires/Swarm_Tyr/T_Tire_Swarm_Tyr_D}
  NamedRerouteDeclaration(None=@n22.A) {name=TireDiffuse_A, nodeColor=<rgba 0.1719 0 1 1>}
  n23 = TextureSampleParameter2D() {group=Rim Properties, parameterName=RimDiffuse, texture=/VehicleCosmetics/Wheels/Blender/Textures/T_Rim_Blender_BC}
  NamedRerouteDeclaration(None=@n23.A) {name=RimeDiffuse_A, nodeColor=<rgba 1 0 0.8203 1>}
  NamedRerouteDeclaration(None=@n12) {name=Rim_InvertColorMask, nodeColor=<rgba 0.0312 1 0 1>}
  n24 = ChannelMaskParameter(None=@n10.RGBA) {group=Rim Color Control, maskChannel=Alpha, parameterName=Rim_AOChannelPicker, sortPriority=2, defaultValue=<rgba 0 0 0 1>}
  NamedRerouteDeclaration(None=@n24.R) {name=Rim_AOChannelPicker, nodeColor=<rgba 0 1 0.3438 1>}
  NamedRerouteDeclaration(None=@n3.A) {name=Tire_CustomColorMask, nodeColor=<rgba 0 0.7188 1 1>}
  NamedRerouteDeclaration(None=@n23.RGB) {name=RimDiffuse_RGB, nodeColor=<rgba 0 0.9766 1 1>}
  NamedRerouteDeclaration(None=@n22.RGB) {name=TireDiffuse_RGB, nodeColor=<rgba 0 0.7422 1 1>}
  NamedRerouteDeclaration(None=VectorParameter() {parameterName=TireColor, defaultValue=<rgba 0.015 0.015 0.015 0>}.RGB) {name=TireColor_RGB, nodeColor=<rgba 0 1 0.8594 1>}
  NamedRerouteDeclaration(None=VectorParameter() {group=Custom Color Properties, parameterName=CustomColor, defaultValue=<rgba 1 1 1 0>}.RGB) {name=CustomColor_RGB, nodeColor=<rgba 0 1 0.5078 1>}
  NamedRerouteDeclaration(None=VertexColor().R) {name=VertColor_R, nodeColor=<rgba 0 0.4141 1 1>}
  n25 = OneMinus(None=Fresnel() {baseReflectFraction=0.01, exponent=0.9})
  n26 = Add(A=Saturate(None=ConstantBiasScale(None=@n25) {bias=-0.4, scale=2})) {constB=0.7}
  NamedRerouteDeclaration(None=Saturate(None=Multiply(A=@n26) {constB=1.2})) {name=Fresnel_Mask, nodeColor=<rgba 0 1 0.3438 1>}
  n27 = TextureSampleParameter2D() {
   group=Rim Properties
   parameterName=RimNormal
   samplerType=SAMPLERTYPE_Normal
   texture=/VehicleCosmetics/Wheels/Blender/Textures/T_Rim_Blender_N
  }
  n28 = LinearInterpolate(Alpha=Saturate(None=ScalarParameter() {defaultValue=1, parameterName=TreadUVDirection_X})) {constA=-1}
  n29 = LinearInterpolate(Alpha=Saturate(None=ScalarParameter() {defaultValue=1, parameterName=TreadUVDirection_Y})) {constA=-1}
  n30 = TextureSampleParameter2D(UVs=Multiply(A=TextureCoordinate(), B=AppendVector(A=@n28, B=@n29))) {
   group=Tire Properties
   parameterName=TireNormal
   samplerType=SAMPLERTYPE_Normal
   texture=/VehicleCosmetics/Wheels/_SharedTextures/Tires/Swarm_Tyr/T_Tire_Swarm_Tyr_N
  }
  n31 = Multiply(A=@n30.R, B=@n28) {constB=-1}
  n32 = Multiply(A=@n30.G, B=@n29) {constB=-1}
  NamedRerouteDeclaration(None=LinearInterpolate(
    A=@n27.RGB
    B=MaterialFunctionCall(R=@n31, G=@n32, B=@n30.B) {materialFunction=&AppendMany}.RGB
    Alpha=NamedRerouteUsage()
   )) {name=Normal_Out, nodeColor=<rgba 0.3125 1 0 1>}
  n33 = TextureSampleParameter2D() {
   group=Rim Properties
   parameterName=Rim_EmissiveMask
   samplerType=SAMPLERTYPE_LinearColor
   texture=&T_Rim_Blender_RGB
  }
  n34 = ChannelMaskParameter(None=@n33.RGBA) {group=Rim Properties, parameterName=Rim_EmissiveChannelPicker, sortPriority=2}
  n35 = Multiply(
   A=StaticSwitchParameter(True=OneMinus(None=@n34.R), False=@n34.R) {group=Rim Properties, parameterName=Rim_InvertEmissiveMask, sortPriority=2}
   B=Multiply(A=NamedRerouteUsage(), B=ScalarParameter() {defaultValue=1, group=Rim Properties, parameterName=Rim_EmissiveMultiplier, sortPriority=3})
  )
  n36 = Constant()
  n37 = ChannelMaskParameter(None=TextureSampleParameter2D() {group=Tire Properties, parameterName=Tire_EmissiveMask, texture=/VehicleCosmetics/Wheels/OEM/Textures/Tire_Swarm_Tyr_D}.RGBA) {group=Tire Properties, parameterName=Tire_EmissiveChannelPicker, sortPriority=2}
  n38 = Multiply(
   A=StaticSwitchParameter(True=OneMinus(None=@n37.R), False=@n37.R) {group=Tire Properties, parameterName=Tire_InvertEmissiveMask, sortPriority=2}
   B=Multiply(A=NamedRerouteUsage(), B=ScalarParameter() {defaultValue=1, group=Tire Properties, parameterName=Tire_EmissiveMultiplier, sortPriority=3})
  )
  n39 = Constant()
  NamedRerouteDeclaration(None=LinearInterpolate(
    A=StaticSwitchParameter(True=@n35, False=@n36) {group=Rim Properties, parameterName=UseRimEmissive, sortPriority=1}
    B=StaticSwitchParameter(True=@n38, False=@n39) {group=Tire Properties, parameterName=UseTireEmissive, sortPriority=1}
    Alpha=NamedRerouteUsage()
   )) {name=Emissive_Out, nodeColor=<rgba 1 0.1172 0 1>}
  n40 = VertexNormalWS()
  n41 = ScalarParameter() {parameterName=TirePop}
  n42 = Subtract(A=ObjectPositionWS(), B=WorldPosition().XYZ)
  n43 = VertexColor()
  NamedRerouteDeclaration(None=LinearInterpolate(A=Multiply(A=@n40) {constB=0}, B=Multiply(A=@n41, B=@n42), Alpha=Saturate(None=@n43.G))) {name=WPO_Out, nodeColor=<rgba 0.9688 0 1 1>}
  n44 = StaticSwitchParameter(
   True=NamedRerouteUsage()
   False=LinearInterpolate(
    A=ScalarParameter() {defaultValue=0.5, group=Rim Properties, parameterName=RimRoughness, sortPriority=1}
    B=ScalarParameter() {defaultValue=0.5, group=Rim Properties, parameterName=RimRoughness_Painted, sortPriority=1}
    Alpha=NamedRerouteUsage()
   )
  ) {parameterName=RimRoughnessFromDiffAlpha}
  n45 = StaticSwitchParameter(True=NamedRerouteUsage(), False=ScalarParameter() {defaultValue=0.85, group=Tire Properties, parameterName=TireRoughness, sortPriority=1}) {parameterName=TireRoughnessFromDiffAlpha}
  n46 = NamedRerouteUsage()
  NamedRerouteDeclaration(None=LinearInterpolate(B=LinearInterpolate(A=@n44, B=@n45, Alpha=@n46), Alpha=NamedRerouteUsage()) {constA=0.9}) {name=Roughness_Out, nodeColor=<rgba 1 0.3047 0 1>}
  n47 = ScalarParameter() {defaultValue=1, group=Rim Properties, parameterName=RimSpecular, sortPriority=1}
  n48 = ScalarParameter() {group=Tire Properties, parameterName=TireSpecular, sortPriority=1}
  n49 = NamedRerouteUsage()
  NamedRerouteDeclaration(None=LinearInterpolate(B=LinearInterpolate(A=@n47, B=@n48, Alpha=@n49), Alpha=NamedRerouteUsage()) {constA=0.1}) {name=Spec_Out, nodeColor=<rgba 0.2656 0 1 1>}
  n50 = ScalarParameter() {group=Rim Properties, parameterName=RimMetallic, sortPriority=1}
  n51 = ScalarParameter() {defaultValue=1, group=Rim Properties, parameterName=RimMetallic_Painted, sortPriority=1}
  n52 = NamedRerouteUsage()
  NamedRerouteDeclaration(None=LinearInterpolate(
    A=LinearInterpolate(A=@n50, B=@n51, Alpha=@n52)
    B=ScalarParameter() {group=Tire Properties, parameterName=TireMetallic, sortPriority=1}
    Alpha=NamedRerouteUsage()
   )) {name=Metallic_Out, nodeColor=<rgba 0 1 0.6719 1>}
  n53 = NamedRerouteUsage()
  n54 = Constant() {r=0.75}
  n55 = ConstantBiasScale(None=OneMinus(None=Fresnel() {baseReflectFraction=-0.6, exponent=0.4})) {bias=-0.3, scale=1}
  n56 = Saturate(None=Add(A=Saturate(None=@n55)) {constB=0.1})
  MaterialExpressionReroute_3 = Reroute(None=@n56)
  n57 = Multiply(A=LinearInterpolate(A=NamedRerouteUsage(), B=NamedRerouteUsage(), Alpha=NamedRerouteUsage()), B=@n19)
  NamedRerouteDeclaration(None=LinearInterpolate(A=Desaturation(None=@n53, Fraction=@n54), B=Multiply(A=@MaterialExpressionReroute_3.Output, B=@n57), Alpha=NamedRerouteUsage())) {name=Diffuse_Out, nodeColor=<rgba 0 1 1 1>}
  PinBase()
  PinBase!(Output=@n56) {reroutePins [Output {expression=&self:MaterialExpressionReroute_3}]}
 ]}
